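# Looping interior crystal cave ambience cue for the Herrald's Fortress point of interest.
# A quality-level node plays a looping quad wave for the first two levels and a looping stereo wave for the last two.
# It uses Athena ambient attenuation but is routed to the Fort_Doors sound class.
# A concurrency override allows one instance per owner and prevents new ones while it plays.
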SoundCue /Game/Sounds/Artemis/POI/HerraldsFortress/Ambience_Chrome_HerraldsFortress_CrystalCave_Loop_Cue
root=QualityLevel_0
nodes [
 {id=QualityLevel_0, children [WavePlayer_1,WavePlayer_1,WavePlayer_0,WavePlayer_0]}
 {
  id=WavePlayer_0
  properties {bLooping=true, soundWaveAssetPtr=&/Ambience_Chrome_HerraldsFortress_CrystalCave_Int_Loop_Stereo}
 }
 {
  id=WavePlayer_1
  properties {bLooping=true, soundWaveAssetPtr=&/Ambience_Chrome_HerraldsFortress_CrystalCave_Int_Loop_Quad}
 }
]
properties {
 attenuationSettings=&Fort_Ambient_Athena_Att
 bOverrideConcurrency=true
 maxDistance=2097152
 soundClassObject=&Fort_Doors
 volumeMultiplier=0.9
 concurrencyOverrides {bLimitToOwner=true, maxCount=1, resolutionRule=PreventNew, retriggerTime=2}
}
