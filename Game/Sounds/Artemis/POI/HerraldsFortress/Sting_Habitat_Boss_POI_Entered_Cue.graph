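# Music sting cue for the Herrald's Fortress habitat boss POI, named for playing when a player enters it.
# A single wave player feeds the root with the Habitat POI-discovered sting, slightly attenuated in volume and routed to the POI music sound class.
# A concurrency override allows one instance and blocks new plays while it is active, so the sting cannot stack.
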SoundCue /Game/Sounds/Artemis/POI/HerraldsFortress/Sting_Habitat_Boss_POI_Entered_Cue
root=WavePlayer_1
nodes [{id=WavePlayer_1, properties {soundWaveAssetPtr=&/Habitat_Sting_POI_Discovered}}]
properties {
 bOverrideConcurrency=true
 maxDistance=2097152
 soundClassObject=/Game/Sounds/FortSoundClassesAndModes/Fort_Music_POI_PSM
 volumeMultiplier=0.9
 concurrencyOverrides {maxCount=1, resolutionRule=PreventNew}
}
